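; Special variant Blueprint of the DarkNexus street info board prop that can report which players are emoting near it.
; GetPlayersEmotingInBox gathers Fortnite player pawns overlapping the Box component, checks each pawn's controller for an active emote, and returns the matching pawns.
; The construction script only calls the parent, and nothing shown here calls the query or acts on its result.
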(fn GetPlayersEmotingInBox
  (for _array_element (Collision|GetOverlappingActors (Variables|Default|GetBox) "/Script/FortniteGame.FortPlayerPawnAthena" :!class /Script/Engine.PrimitiveComponent)
    (bind _asfort_player_pawn_athena (Utilities|Casting|CastToFortPlayerPawnAthena _array_element))
    (bind _asfort_player_controller_athena (Utilities|Casting|CastToFortPlayerControllerAthena (Pawn|GetController _asfort_player_pawn_athena)))
    (if (Emote|IsPlayingEmote _asfort_player_controller_athena)
      (Utilities|Array|Add (LocalVariables|GetLocalPlayersEmotting) _asfort_player_pawn_athena)))
  (return (LocalVariables|GetLocalPlayersEmotting)))

(fn ConstructionScript
  (|CallParentFunction))
